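# Physics asset for the SwiftKestrel backpack cosmetic, defining collision bodies and joints for the pack and its small dangling chains.
# Two overlapping boxes form the pack body, with tiny base capsules locked rotationally to it as anchors for the dyn_chain capsules.
# Chain00 swings as two linked segments, while both chain01 segments swing from separate anchors, all with narrow swing and zero-degree twist limits.
# Chain02 is fully rotation-locked to its anchor.
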PhysicsAsset /BRCosmetics/Accessories/FORT_Backpacks/Backpack_M_MED_SwiftKestrel/Meshes/M_MED_SwiftKestrel_Pack_Physics
bodies {
 base_chain00_00 [{shapeType=Capsule, length=0.505, radius=0.505, shapeName=base_chain00_00_capsule, rotation=<rot 0 -90 0>}]
 base_chain01_00 [{shapeType=Capsule, length=0.505, radius=0.505, shapeName=base_chain01_00_capsule, rotation=<rot 0 -90 0>}]
 base_chain01_01 [{shapeType=Capsule, length=0.505, radius=0.505, shapeName=base_chain01_01_capsule, rotation=<rot 0 -90 0>}]
 base_chain02_00 [{shapeType=Capsule, length=0.505, radius=0.505, shapeName=base_chain02_00_capsule, rotation=<rot 0 -90 0>}]
 dyn_chain00_00 [{shapeType=Capsule, length=5.0205, radius=0.5189, shapeName=dyn_chain00_00_capsule, center=<2.9826 0 0>, rotation=<rot 90 0 0>}]
 dyn_chain00_01 [{shapeType=Capsule, length=3.2057, radius=0.7171, shapeName=dyn_chain00_01_capsule, center=<4.2694 0 0>, rotation=<rot 90 0 0>}]
 dyn_chain01_00 [{shapeType=Capsule, length=5, radius=0.5, shapeName=dyn_chain01_00_capsule, center=<2.7762 0 0>, rotation=<rot 90 0 0>}]
 dyn_chain01_01 [{shapeType=Capsule, length=3.2843, radius=0.5, shapeName=dyn_chain01_00_capsule, center=<2.4517 0 0>, rotation=<rot 90 0 0>}]
 dyn_chain02_00 [{shapeType=Capsule, length=0.9598, radius=0.5484, shapeName=dyn_chain02_00_capsule, center=<2 0 0>, rotation=<rot 90 0 0>}]
 pack [
  {
   shapeType=Box
   extentX=6.6821
   extentY=9.9514
   extentZ=34.5101
   shapeName=pack_box
   center=<-3.7495 -2.7031 -29.3643>
   rotation=<rot -0.1226 -90 27.7044>
  }
  {
   shapeType=Box
   extentX=6.6821
   extentY=9.9514
   extentZ=34.5101
   shapeName=pack_box
   center=<-8.23 -0.6295 -27.0072>
   rotation=<rot -0.1226 -90 27.7044>
  }
 ]
}
constraints [
 {bone1Name=dyn_chain00_00, bone2Name=base_chain00_00, swing1LimitDegrees=20, swing2LimitDegrees=15, twistLimitDegrees=0}
 {bone1Name=dyn_chain00_01, bone2Name=dyn_chain00_00, swing1LimitDegrees=30, swing2LimitDegrees=15, twistLimitDegrees=0}
 {bone1Name=dyn_chain01_00, bone2Name=base_chain01_00, swing1LimitDegrees=10, swing2LimitDegrees=15, twistLimitDegrees=0}
 {
  bone1Name=dyn_chain02_00
  bone2Name=base_chain02_00
  swing1LimitDegrees=10
  swing1Motion=Locked
  swing2LimitDegrees=15
  swing2Motion=Locked
  twistLimitDegrees=0
  twistMotion=Locked
 }
 {bone1Name=dyn_chain01_01, bone2Name=base_chain01_01, swing1LimitDegrees=10, swing2LimitDegrees=5, twistLimitDegrees=0}
 {
  bone1Name=base_chain00_00
  bone2Name=pack
  swing1LimitDegrees=45
  swing1Motion=Locked
  swing2LimitDegrees=45
  swing2Motion=Locked
  twistLimitDegrees=45
  twistMotion=Locked
 }
 {
  bone1Name=base_chain02_00
  bone2Name=pack
  swing1LimitDegrees=45
  swing1Motion=Locked
  swing2LimitDegrees=45
  swing2Motion=Locked
  twistLimitDegrees=45
  twistMotion=Locked
 }
 {
  bone1Name=base_chain01_00
  bone2Name=pack
  swing1LimitDegrees=45
  swing1Motion=Locked
  swing2LimitDegrees=45
  swing2Motion=Locked
  twistLimitDegrees=45
  twistMotion=Locked
 }
 {
  bone1Name=base_chain01_01
  bone2Name=pack
  swing1LimitDegrees=45
  swing1Motion=Locked
  swing2LimitDegrees=45
  swing2Motion=Locked
  twistLimitDegrees=45
  twistMotion=Locked
 }
]
